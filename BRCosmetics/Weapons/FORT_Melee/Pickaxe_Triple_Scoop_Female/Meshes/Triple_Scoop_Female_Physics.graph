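# Physics asset for the Triple Scoop Female melee pickaxe mesh, defining collision bodies and joints for its two dynamic helmet pieces.
# A small capsule covers the Bat_main bone, and a larger, near-identical capsule covers each of the dyn_helmet_01 and dyn_helmet_02 bones.
# Each helmet body is constrained to Bat_main with limited swing and tighter twist, with slightly wider swing on the second helmet.
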PhysicsAsset /BRCosmetics/Weapons/FORT_Melee/Pickaxe_Triple_Scoop_Female/Meshes/Triple_Scoop_Female_Physics
bodies {
 Bat_main [{shapeType=Capsule, length=8.25, radius=4.5, center=<-0.0007 0 -4.072>}]
 dyn_helmet_01 [{shapeType=Capsule, length=16.4077, radius=16.0882, center=<1.123 0.0172 -0.2906>, rotation=<rot 0 90.1561 -58.4117>}]
 dyn_helmet_02 [{shapeType=Capsule, length=16.401, radius=16.0809, center=<1.172 -0.0118 -0.3246>, rotation=<rot 0 89.8908 -58.5025>}]
}
constraints [
 {bone1Name=dyn_helmet_02, bone2Name=Bat_main, swing1LimitDegrees=35, swing2LimitDegrees=35, twistLimitDegrees=15}
 {bone1Name=dyn_helmet_01, bone2Name=Bat_main, swing1LimitDegrees=30, swing2LimitDegrees=30, twistLimitDegrees=15}
]
